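# Default Lit material for the Rover emote's book prop that reveals a page image on it, usable on skeletal meshes.
# A page texture is sampled through TextureCropping UVs set by a corners parameter, then blended over the prop's diffuse texture.
# The blend mask combines softened page alpha, the crop mask and an MF_Erosion dissolve driven by the Reveal parameter.
# The prop's normal and mask textures are also sampled.
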Material /BRCosmetics/Athena/Effects/Particles/Emotes/NS_Emote_Rover/Materials/M_Emote_Rover_BookReveal
graph {
 properties {bUsedWithMeshDeformer=true, bUsedWithSkeletalMesh=true, pixelDepthOffsetMode=PDOM_Legacy, usedShadingModels=Default Lit}
 nodes [
  n1 = VectorParameter() {parameterName=Crop Corners, defaultValue=<rgba 0.78 0.3 1 0.62>}
  n2 = MaterialFunctionCall!(UVs, TextureIn, UpperLeftCorner=AppendVector(A=@n1.R, B=@n1.R), LowerRightCorner=AppendVector(A=@n1.B, B=@n1.B), ExportFloat4) {materialFunction=/Engine/Functions/Engine_MaterialFunctions01/Texturing/TextureCropping}
  n3 = TextureSample(UVs=@n2.Crop UVs) {texture=/BRCosmetics/Athena/Effects/Particles/Emotes/NS_Emote_Rover/Textures/T_Emote_Rover_BookPage}
  n4 = Divide(A=@n3.A, B=ScalarParameter() {defaultValue=0.01, parameterName=Mask Softness}) {constB=0.01}
  n5 = MaterialFunctionCall!(
   Gradient=@n3.A
   Erosion Amount=ScalarParameter() {parameterName=Reveal}
   Softness=ScalarParameter() {defaultValue=0.06, parameterName=Reveal Softness}
  ) {materialFunction=/BRCosmetics/Athena/Effects/Particles/Skydiving/NS_Contrail_DarkAndLight/Functions/MF_Erosion}
  LinearInterpolate(
   A=TextureSample() {texture=/BRCosmetics/Animation/Props/Gadgets_Rover/Textures/T_Emote_Prop_Rover_D}.RGB
   B=@n3.RGB
   Alpha=Multiply(A=Saturate(None=@n4), B=Multiply(A=@n2.Crop Mask, B=@n5.Increased Range))
  )
  Multiply(A=ScalarParameter() {defaultValue=0.9, parameterName=Blank Page Brightness Multiplier}, B=@n3.RGB)
  TextureSample() {samplerType=SAMPLERTYPE_Normal, texture=/BRCosmetics/Animation/Props/Gadgets_Rover/Textures/T_Emote_Prop_Rover_N}
  TextureSample() {samplerType=SAMPLERTYPE_Masks, texture=/BRCosmetics/Animation/Props/Gadgets_Rover/Textures/T_Emote_Prop_Rover_S}
 ]
}
